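# Sound cue for the reskinned Deimos Brute monster's death vocalization.
# A Random node picks one of five equally weighted Brute death wave players and is excluded from branch culling.
# Playback uses close Deimos vocal attenuation, the creature vocalization sound class, a boosted volume and global plus local Deimos vocal concurrency sets.
# Virtualization is disabled and priority bypasses volume scaling.
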SoundCue /Game/Sounds/Fort_Monster_Deimos/Brute/Redesign/Brute_Reskin_VO_Death_Cue
root=Random_0
nodes [
 {
  id=Random_0
  children [WavePlayer_0,WavePlayer_1,WavePlayer_2,WavePlayer_3,WavePlayer_7]
  properties {bShouldExcludeFromBranchCulling=true, weights [1,1,1,1,1]}
 }
 {id=WavePlayer_0, properties {soundWaveAssetPtr=&/DEIMOS_2021_Brute_Death_010}}
 {id=WavePlayer_1, properties {soundWaveAssetPtr=&/DEIMOS_2021_Brute_Death_011}}
 {id=WavePlayer_2, properties {soundWaveAssetPtr=&/DEIMOS_2021_Brute_Death_012}}
 {id=WavePlayer_3, properties {soundWaveAssetPtr=&/DEIMOS_2021_Brute_Death_013}}
 {id=WavePlayer_7, properties {soundWaveAssetPtr=&/DEIMOS_2021_Brute_Death_014}}
]
properties {
 attenuationSettings=/Game/Sounds/Attenuation/Fort_Deimos_Vocal_Close_Att
 bBypassVolumeScaleForPriority=true
 maxDistance=13000
 priority=0.75
 soundClassObject=&Fort_Creatures_Vocalization_Tactical
 virtualizationMode=Disabled
 volumeMultiplier=3.5919
 concurrencyOverrides {resolutionRule=PreventNew}
 concurrencySet [/Game/Sounds/Fort_Monster_Deimos/Settings/Deimos_Vocal_Global_Conc, /Game/Sounds/Fort_Monster_Deimos/Settings/Deimos_Vocal_Local_Conc]
}
