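# Sound cue for pain vocalizations of the reskinned Deimos Brute monster, layered by distance.
# A mixer blends a louder close layer of six redesigned pain waves with a quieter distant layer of five older distant pain waves.
# Each layer picks at random, passes through a modulator, and gets its own close or distant vocal attenuation, the distant one overriding falloff and shape.
# Global and local Deimos vocal concurrency limit playback, and virtualization is disabled.
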SoundCue /Game/Sounds/Fort_Monster_Deimos/Brute/Redesign/Brute_Reskin_VO_Pain_Cue
root=Mixer_0
nodes [
 {id=Attenuation_0, children [Modulator_0], properties {attenuationSettings=/Game/Sounds/Attenuation/Fort_Deimos_Vocal_Close_Att}}
 {
  id=Attenuation_1
  children [Modulator_1]
  properties {
   attenuationSettings=/Game/Sounds/Attenuation/Fort_Deimos_Vocal_Distant_Att
   attenuationOverrides {falloffDistance=7000, stereoSpread=6000, attenuationShapeExtents {x=5000}}
  }
 }
 {id=Mixer_0, children [Attenuation_0, Attenuation_1], properties {inputVolume [1.5,0.7]}}
 {id=Modulator_0, children [Random_0]}
 {id=Modulator_1, children [Random_3], properties {pitchMax=1.2}}
 {
  id=Random_0
  children [WavePlayer_0,WavePlayer_1,WavePlayer_2,WavePlayer_3,WavePlayer_4,WavePlayer_5]
  properties {bShouldExcludeFromBranchCulling=true, weights [1,1,1,1,1,1]}
 }
 {
  id=Random_3
  children [WavePlayer_15, WavePlayer_16, WavePlayer_17, WavePlayer_18, WavePlayer_19]
  properties {bShouldExcludeFromBranchCulling=true, weights [1,1,1,1,1]}
 }
 {id=WavePlayer_0, properties {soundWaveAssetPtr=&/DEIMOS_2021_Brute_Pain_010}}
 {id=WavePlayer_1, properties {soundWaveAssetPtr=&/DEIMOS_2021_Brute_Pain_011}}
 {id=WavePlayer_15, properties {soundWaveAssetPtr=/Game/Sounds/Fort_Monster_Deimos/Brute/Brute_Pain_Distant_01}}
 {id=WavePlayer_16, properties {soundWaveAssetPtr=/Game/Sounds/Fort_Monster_Deimos/Brute/Brute_Pain_Distant_02}}
 {id=WavePlayer_17, properties {soundWaveAssetPtr=/Game/Sounds/Fort_Monster_Deimos/Brute/Brute_Pain_Distant_03}}
 {id=WavePlayer_18, properties {soundWaveAssetPtr=/Game/Sounds/Fort_Monster_Deimos/Brute/Brute_Pain_Distant_04}}
 {id=WavePlayer_19, properties {soundWaveAssetPtr=/Game/Sounds/Fort_Monster_Deimos/Brute/Brute_Pain_Distant_05}}
 {id=WavePlayer_2, properties {soundWaveAssetPtr=&/DEIMOS_2021_Brute_Pain_012}}
 {id=WavePlayer_3, properties {soundWaveAssetPtr=&/DEIMOS_2021_Brute_Pain_013}}
 {id=WavePlayer_4, properties {soundWaveAssetPtr=&/DEIMOS_2021_Brute_Pain_014}}
 {id=WavePlayer_5, properties {soundWaveAssetPtr=&/DEIMOS_2021_Brute_Pain_015}}
]
properties {
 attenuationSettings=&Fort_Deimos_Vocal_Close_Att
 bBypassVolumeScaleForPriority=true
 maxDistance=24000
 priority=0.75
 soundClassObject=&Fort_Creatures_Vocalization_Tactical
 virtualizationMode=Disabled
 volumeMultiplier=1.7955
 attenuationOverrides {distanceAlgorithm=NaturalSound, falloffDistance=4000, attenuationShapeExtents {x=1000}}
 concurrencyOverrides {resolutionRule=PreventNew}
 concurrencySet [/Game/Sounds/Fort_Monster_Deimos/Settings/Deimos_Vocal_Global_Conc, /Game/Sounds/Fort_Monster_Deimos/Settings/Deimos_Vocal_Local_Conc]
}
