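# Sound cue for the redesigned Deimos Brute monster's breathing vocalization while running.
# A Random node picks one of four equally weighted running-breath waves, and a Modulator at the root varies each playback.
# It uses the creature vocalization sound class, a close Deimos vocal attenuation, and shared global and local vocal concurrency.
# Virtualization is disabled, and priority is set to bypass volume scaling.
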SoundCue /Game/Sounds/Fort_Monster_Deimos/Brute/Redesign/Brute_Reskin_VO_Run_Breath_Cue
root=Modulator_1
nodes [
 {id=Modulator_1, children [Random_0]}
 {
  id=Random_0
  children [WavePlayer_0,WavePlayer_1,WavePlayer_2,WavePlayer_3]
  properties {bShouldExcludeFromBranchCulling=true, weights [1,1,1,1]}
 }
 {id=WavePlayer_0, properties {soundWaveAssetPtr=&/DEIMOS_2021_Brute_RunningBreaths_010}}
 {id=WavePlayer_1, properties {soundWaveAssetPtr=&/DEIMOS_2021_Brute_RunningBreaths_011}}
 {id=WavePlayer_2, properties {soundWaveAssetPtr=&/DEIMOS_2021_Brute_RunningBreaths_012}}
 {id=WavePlayer_3, properties {soundWaveAssetPtr=&/DEIMOS_2021_Brute_RunningBreaths_013}}
]
properties {
 attenuationSettings=/Game/Sounds/Attenuation/Fort_Deimos_Vocal_Close_Att
 bBypassVolumeScaleForPriority=true
 maxDistance=13000
 priority=0.75
 soundClassObject=/Game/Sounds/FortSoundClassesAndModes/Fort_Creatures_Vocalization_Normal
 virtualizationMode=Disabled
 volumeMultiplier=2.993
 attenuationOverrides {distanceAlgorithm=NaturalSound, falloffDistance=4000, attenuationShapeExtents {x=1000}}
 concurrencyOverrides {maxCount=2, volumeScale=0.75, volumeScaleAttackTime=0.1}
 concurrencySet [/Game/Sounds/Fort_Monster_Deimos/Settings/Deimos_Vocal_Global_Conc, /Game/Sounds/Fort_Monster_Deimos/Settings/Deimos_Vocal_Local_Conc]
}
