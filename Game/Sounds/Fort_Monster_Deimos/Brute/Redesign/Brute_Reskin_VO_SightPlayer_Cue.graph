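# Sound cue for the reskinned Deimos Brute's vocalization when it sights a player.
# A short random delay feeds a modulator, then a Random node picks one of five equally weighted SightedPlayer waves and is excluded from branch culling.
# It plays through a tactical creature vocalization class with sight-player attenuation and global and local Deimos vocal concurrency sets that limit overlapping voices.
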SoundCue /Game/Sounds/Fort_Monster_Deimos/Brute/Redesign/Brute_Reskin_VO_SightPlayer_Cue
root=Delay_0
nodes [
 {id=Delay_0, children [Modulator_0], properties {delayMax=0.75, delayMin=0.25}}
 {id=Modulator_0, children [Random_2]}
 {
  id=Random_2
  children [WavePlayer_5, WavePlayer_6, WavePlayer_12, WavePlayer_13, WavePlayer_14]
  properties {bShouldExcludeFromBranchCulling=true, weights [1,1,1,1,1]}
 }
 {id=WavePlayer_12, properties {soundWaveAssetPtr=&/DEIMOS_2021_Brute_SightedPlayer_012}}
 {id=WavePlayer_13, properties {soundWaveAssetPtr=&/DEIMOS_2021_Brute_SightedPlayer_013}}
 {id=WavePlayer_14, properties {soundWaveAssetPtr=&/DEIMOS_2021_Brute_SightedPlayer_014}}
 {id=WavePlayer_5, properties {soundWaveAssetPtr=&/DEIMOS_2021_Brute_SightedPlayer_010}}
 {id=WavePlayer_6, properties {soundWaveAssetPtr=&/DEIMOS_2021_Brute_SightedPlayer_011}}
]
properties {
 attenuationSettings=/Game/Sounds/Attenuation/Fort_Deimos_SightPlayer_Att
 bBypassVolumeScaleForPriority=true
 maxDistance=5500
 priority=0.75
 soundClassObject=&Fort_Creatures_Vocalization_Tactical
 volumeMultiplier=4.041
 attenuationOverrides {distanceAlgorithm=NaturalSound, falloffDistance=4200, attenuationShapeExtents {x=800}}
 concurrencyOverrides {maxCount=3, resolutionRule=StopFarthestThenPreventNew, volumeScale=0.75, volumeScaleAttackTime=0.1}
 concurrencySet [/Game/Sounds/Fort_Monster_Deimos/Settings/Deimos_Vocal_Global_Conc, /Game/Sounds/Fort_Monster_Deimos/Settings/Deimos_Vocal_Local_Conc]
}
