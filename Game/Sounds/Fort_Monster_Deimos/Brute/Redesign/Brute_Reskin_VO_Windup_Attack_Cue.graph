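# Sound cue for the reskinned Deimos Brute monster's vocalization during an attack windup.
# A Random node picks among five equally weighted melee attack waves and feeds a Modulator at the root.
# It plays as a high-priority, boosted tactical creature vocalization with close vocal attenuation and virtualization disabled.
# Global and local Deimos vocal concurrency sets limit overlapping voices.
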SoundCue /Game/Sounds/Fort_Monster_Deimos/Brute/Redesign/Brute_Reskin_VO_Windup_Attack_Cue
root=Modulator_0
nodes [
 {id=Modulator_0, children [Random_2]}
 {
  id=Random_2
  children [WavePlayer_0,WavePlayer_1,WavePlayer_2,WavePlayer_3,WavePlayer_4]
  properties {bShouldExcludeFromBranchCulling=true, weights [1,1,1,1,1]}
 }
 {id=WavePlayer_0, properties {soundWaveAssetPtr=&/DEIMOS_2021_Brute_MeleeAttack_014}}
 {id=WavePlayer_1, properties {soundWaveAssetPtr=&/DEIMOS_2021_Brute_MeleeAttack_010}}
 {id=WavePlayer_2, properties {soundWaveAssetPtr=&/DEIMOS_2021_Brute_MeleeAttack_011}}
 {id=WavePlayer_3, properties {soundWaveAssetPtr=&/DEIMOS_2021_Brute_MeleeAttack_012}}
 {id=WavePlayer_4, properties {soundWaveAssetPtr=&/DEIMOS_2021_Brute_MeleeAttack_013}}
]
properties {
 attenuationSettings=/Game/Sounds/Attenuation/Fort_Deimos_Vocal_Close_Att
 bBypassVolumeScaleForPriority=true
 maxDistance=13000
 priority=0.9
 soundClassObject=&Fort_Creatures_Vocalization_Tactical
 virtualizationMode=Disabled
 volumeMultiplier=2.694
 attenuationOverrides {distanceAlgorithm=NaturalSound, falloffDistance=4200, attenuationShapeExtents {x=800}}
 concurrencyOverrides {resolutionRule=PreventNew}
 concurrencySet [/Game/Sounds/Fort_Monster_Deimos/Settings/Deimos_Vocal_Global_Conc, /Game/Sounds/Fort_Monster_Deimos/Settings/Deimos_Vocal_Local_Conc]
}
